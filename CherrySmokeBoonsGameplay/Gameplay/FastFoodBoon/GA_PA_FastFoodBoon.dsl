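; Fast Food Boon gameplay ability that grants health or shield when the player picks up restoration items.
; Each pickup event snapshots inventory size, classifies the item by restoration tags and runs a temporary polling timer until the size changes or a delay expires.
; A change applies SetByCaller healing or shielding effects per picked-up count, scaled by hero rank tag, without removing the item.
; Item-consuming meat and small-potion functions remain but are uncalled.
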(event Custom|InventoryChanged
  (bind _currentherorank (CallFunction|GetHeroRank))
  (Variables|Default|SetCurrentHeroRank _currentherorank)
  (switch Utilities|FlowControl|Switch|SwitchonFastFoodBoonEnum_V2 (Variables|Default|GetCurrentItemType)
    (:NewEnumerator14
      (Use|UseHealingItem))
    (:NewEnumerator15
      (Use|UseShieldingItem))
    (:NewEnumerator16
      (Development|PrintString "Mom says we have fast food at home :(((" true true "(R=1.000000,G=0.100000,B=0.100000,A=1.000000)" 3.0))))

(event Custom|InventoryScanner
  ; TEMP
  (CallFunction|ScanforInventoryChanges))

(event Ability|EventActivateAbility
  (|CallParentFunction)
  (Variables|Default|SetPlayerPawn (CallFunction|GetAvatarasFortPlayerPawn))
  (Variables|Default|SetHealingSpec (Ability|MakeOutgoingGameplayEffectSpec "/CherrySmokeBoonsGameplay/Gameplay/FastFoodBoon/GE_PA_FastFoodBoon_Healing.GE_PA_FastFoodBoon_Healing_C"))
  (Variables|Default|SetShieldingSpec (Ability|MakeOutgoingGameplayEffectSpec "/CherrySmokeBoonsGameplay/Gameplay/FastFoodBoon/GE_PA_FastFoodBoon_Shielding.GE_PA_FastFoodBoon_Shielding_C"))
  (bind (_asynctaskproxy _payload) (Ability|Async|WaitGameplayEventToActor (Variables|Default|GetPlayerPawn) "(TagName=\"Event.CollectedPickup\")")
    (:then
      (Variables|Default|SetAsyncActionA _asynctaskproxy)
      ; TEMP
      (bind _returnvalue (Utilities|Time|SetTimerbyEvent (delegate Custom|InventoryScanner) 0.01 true true))
      (bind _output_get (Variables|Default|SetInventorySensorTimer _returnvalue))
      (Utilities|Time|PauseTimerbyHandle _output_get))
    (:EventReceived
      (CallFunction|StartInventoryScan)
      (Utilities|FlowControl|Sequence
        (:then_0
          (bind _asfort_pickup_athena (Utilities|Casting|CastToFortPickupAthena (.Target (Utilities|Struct|BreakGameplayEventData _payload))))
          (bind _output_get_2 (Variables|Default|SetCurrentItemEntry (Pickup|GetPrimaryItemEntry _asfort_pickup_athena)))
          (bind _itemenum (CallFunction|ItemSelector _output_get_2))
          (Variables|Default|SetCurrentItemType _itemenum))
        (:then_1
          ; TEMP
          (Utilities|FlowControl|Delay 2.0)
          (Utilities|Time|PauseTimerbyHandle (Variables|Default|GetInventorySensorTimer)))))))

(event Custom|InventorySensor
  ; OLD VERSION
  ;; TEMP
  (CallFunction|ScanforInventoryChanges))

(event Ability|EventOnEndAbility
  (params (bWasCancelled :type Boolean))
  (|CallParentFunction bWasCancelled)
  (Ability|Async|EndAction (Variables|Default|GetAsyncActionA))
  (Utilities|Time|ClearandInvalidateTimerbyHandle (Variables|Default|GetInventorySensorTimer)))

(fn ApplyHeal
  (params (HealAmount :type Float%20%28double-precision%29))
  (bind _returnvalue (Ability|GameplayEffect|AssignTagSetbyCallerMagnitude (Variables|Default|GetHealingSpec) "(TagName=\"SetByCaller.Healing\")" HealAmount))
  (Ability|ApplyGameplayEffectSpecToOwner _returnvalue)
  (return))

(fn ApplyShield
  (params (ShieldAmount :type Float%20%28double-precision%29))
  (bind _returnvalue (Ability|GameplayEffect|AssignTagSetbyCallerMagnitude (Variables|Default|GetShieldingSpec) "(TagName=\"SetByCaller.Shield\")" ShieldAmount))
  (Ability|ApplyGameplayEffectSpecToOwner _returnvalue)
  (return))

(fn CheckHealthValue
  (return (>= (FortPawn|GetHealth (Ability|GetActivatingPawn)) (FortPawn|GetMaxHealth (Ability|GetActivatingPawn))) (FortPawn|GetHealth (Ability|GetActivatingPawn))))

(fn CheckShieldValue
  (return (>= (FortPawn|GetShield (Ability|GetActivatingPawn)) (FortPawn|GetMaxShield (Ability|GetActivatingPawn))) (FortPawn|GetShield (Ability|GetActivatingPawn))))

(fn DestroyItemInInventory
  (params (AmountToRemove :type Integer))
  (Inventory|RemoveFortItemFromPlayer (Fortnite|GetFortPlayerControllerfromActor (Ability|GetAvatarActorfromActorInfo)) (Variables|Default|GetTargetItem) AmountToRemove)
  (return))

(fn FindItemInInventory
  (bind _currentitementry (Variables|Default|GetCurrentItemEntry))
  (bind (_itemdefinition _itemguid _count) (Class|FortItemFunctionLibrary|BreakFortItemEntry _currentitementry))
  (bind _asfort_player_controller_athena (Utilities|Casting|CastToFortPlayerControllerAthena (Fortnite|GetFortPlayerControllerfromActor (Ability|GetAvatarActorfromActorInfo))))
  (bind _items (InventoryManagement|GetItemsfromInventoryOwner _asfort_player_controller_athena))
  (bind _output_get (Variables|Default|SetInventoryArray _items))
  (Utilities|Array|Reverse _output_get)
  (bind _array_element (Utilities|Array|ForEachLoopwithBreak (Variables|Default|GetInventoryArray))
    (:LoopBody
      (if (GameplayTags|TagContainer|CompareTagContainertoOtherTagContainers (GameplayTags|GetOwnedGameplayTags _array_element :!class /Script/GameplayTags.BlueprintGameplayTagLibrary) (GameplayTags|GetOwnedGameplayTags _itemdefinition :!class /Script/GameplayTags.BlueprintGameplayTagLibrary))
        (Variables|Default|SetTargetItem _array_element)))
    (:Completed
      _))
  (bind _output_get_2 (Variables|Default|SetTargetItemCount _count))
  (return _output_get_2))

(fn GetHeroRank
  (if (GameplayTags|HasAnyMatchingGameplayTags (CallFunction|GetAvatarasFortPlayerPawn) (GameplayTags|MakeLiteralGameplayTagContainer "(GameplayTags=((TagName=\"CherrySmoke.HeroRank.S\")))"))
    (return 3)
    (else
      (if (GameplayTags|HasAnyMatchingGameplayTags (CallFunction|GetAvatarasFortPlayerPawn) (GameplayTags|MakeLiteralGameplayTagContainer "(GameplayTags=((TagName=\"CherrySmoke.HeroRank.A\")))"))
        (return 2)
        (else
          (if (GameplayTags|HasAnyMatchingGameplayTags (CallFunction|GetAvatarasFortPlayerPawn) (GameplayTags|MakeLiteralGameplayTagContainer "(GameplayTags=((TagName=\"CherrySmoke.HeroRank.B\")))"))
            (return 1)
            (else
              (Development|PrintString "Rank C? Confusion..." true true "(R=1.000000,G=0.100000,B=0.100000,A=1.000000)" 3.0)
              (return 1))))))))

(fn GetTagsFromItemEntryStruct
  (params (FortItemEntryStruct :type Fort%20Item%20Entry%20Structure))
  (return (GameplayTags|GetOwnedGameplayTags (Utilities|Struct|BreakItemEntry FortItemEntryStruct) :!class /Script/GameplayTags.BlueprintGameplayTagLibrary)))

(fn IsEdible
  (params (GameplayTagContainer :type Gameplay%20Tag%20Container%20Structure))
  (if (not (GameplayTags|HasTag GameplayTagContainer "(TagName=\"Item.Consumable.ShieldMushroom\")"))
    (return (not (Math|Boolean|ORBoolean (GameplayTags|HasTag GameplayTagContainer "(TagName=\"Item.Consumable.Edible\")") (GameplayTags|HasTag GameplayTagContainer "(TagName=\"Trait.Fruit\")") (GameplayTags|HasTag GameplayTagContainer "(TagName=\"Trait.Veggie\")") (GameplayTags|HasTag GameplayTagContainer "(TagName=\"Item.Consumable.Paprika.Bandage\")") (GameplayTags|HasTag GameplayTagContainer "(TagName=\"Trait.Fish\")"))))
    (else
      (return false))))

(fn IsMeat
  (params (GameplayTagContainer :type Gameplay%20Tag%20Container%20Structure))
  (return (not (and (GameplayTags|HasTag GameplayTagContainer "(TagName=\"Item.Consumable.Meat\")") (GameplayTags|HasTag GameplayTagContainer "(TagName=\"Trait.Meat\")")))))

(fn IsSmallShieldPotion
  (params (GameplayTagContainer :type Gameplay%20Tag%20Container%20Structure))
  (return (not (and (GameplayTags|HasTag GameplayTagContainer "(TagName=\"Item.Consumable.SmallShieldPotion\")") (GameplayTags|HasTag GameplayTagContainer "(TagName=\"Item.Consumable.Paprika.ShieldSmall\")")))))

(fn ItemSelector
  (params (FortItemEntryStruct :type Fort%20Item%20Entry%20Structure))
  (bind _gameplaytagcontainer (CallFunction|GetTagsfromItemEntryStruct FortItemEntryStruct))
  (if (GameplayTags|HasTag _gameplaytagcontainer "(TagName=\"Trait.Restoration.Health\")")
    (return "NewEnumerator14")
    (elif (GameplayTags|HasTag _gameplaytagcontainer "(TagName=\"Trait.Restoration.Shields\")")
      (return "NewEnumerator15")
      (else
        (Development|PrintString "Item Selector is Confusion..." true true "(R=1.000000,G=0.100000,B=0.100000,A=1.000000)" 3.0)
        (return "NewEnumerator16")))))

(fn ScanForInventoryChanges
  (bind _asfort_player_controller_athena (Utilities|Casting|CastToFortPlayerControllerAthena (Fortnite|GetFortPlayerControllerfromActor (Ability|GetAvatarActorfromActorInfo))))
  (bind _items (InventoryManagement|GetItemsfromInventoryOwner _asfort_player_controller_athena))
  (if (!= (Utilities|Array|Length _items) (Variables|Default|GetLastInventorySize))
    (CallFunction|InventoryChanged)
    (Utilities|Time|PauseTimerbyHandle (Variables|Default|GetInventorySensorTimer))
    (Development|PrintString "INVENTORY CHANGVESSEDDD!!!")
    (return)))

(fn SetHealingAmount
  (if (== (Variables|Default|GetCurrentHeroRank) 3)
    (bind _output_get (Variables|Default|SetHealthAmountRankS 2e+01))
    (return _output_get)
    (else
      (if (== (Variables|Default|GetCurrentHeroRank) 2)
        (bind _output_get_2 (Variables|Default|SetHealthAmountRankA 1e+01))
        (return _output_get_2)
        (else
          (if (== (Variables|Default|GetCurrentHeroRank) 1)
            (bind _output_get_3 (Variables|Default|SetHealthAmountRankB 5.0))
            (return _output_get_3)
            (else
              (return 0.0))))))))

(fn SetShieldingAmount
  (if (== (Variables|Default|GetCurrentHeroRank) 3)
    (bind _output_get (Variables|Default|SetShieldAmountRankS 2e+01))
    (return _output_get)
    (else
      (if (== (Variables|Default|GetCurrentHeroRank) 2)
        (bind _output_get_2 (Variables|Default|SetShieldAmountRankA 1e+01))
        (return _output_get_2)
        (else
          (if (== (Variables|Default|GetCurrentHeroRank) 1)
            (bind _output_get_3 (Variables|Default|SetShieldAmountRankB 5.0))
            (return _output_get_3)
            (else
              (return 0.0))))))))

(fn StartInventoryScan
  (Utilities|Array|Clear (Variables|Default|GetItemsScanned))
  (bind _asfort_player_controller_athena (Utilities|Casting|CastToFortPlayerControllerAthena (Fortnite|GetFortPlayerControllerfromActor (Ability|GetAvatarActorfromActorInfo))))
  (bind _items (InventoryManagement|GetItemsfromInventoryOwner _asfort_player_controller_athena))
  (Variables|Default|SetItemsScanned _items)
  (Variables|Default|SetLastInventorySize (Utilities|Array|Length (Variables|Default|GetItemsScanned)))
  (Utilities|Map|Clear (Variables|Default|GetInventoryMap))
  (for _array_element (Variables|Default|GetItemsScanned)
    (Utilities|Map|Add (Variables|Default|GetInventoryMap) _array_element (Item|GetNuminStack _array_element)))
  (Utilities|Time|UnpauseTimerbyHandle (Variables|Default|GetInventorySensorTimer))
  (return))

(fn UseHealingItem
  (bind (_returnvalue _currenthealth) (PureMath|CheckHealthValue))
  (bind _targetitemcount (Shared|FindIteminInventory))
  (LocalVariables|SetFoundItemCount _targetitemcount)
  (Utilities|FlowControl|ForLoopwithBreak 0 (Math|Integer|Clamp(Integer) (- (LocalVariables|GetFoundItemCount) 1) 0 100)
    (:LoopBody
      (bind _healingamount (CallFunction|SetHealingAmount))
      (Shared|ApplyHeal _healingamount)
      (if (>= _currenthealth (Variables|Default|GetHealingCap))
        _))
    (:Completed
      _))
  (return))

(fn UseMeat
  (bind (_returnvalue _currenthealth) (PureMath|CheckHealthValue))
  (bind (_returnvalue_2 _currenthealth_2) (PureMath|CheckHealthValue))
  (bind (_returnvalue_3 _currenthealth_3) (PureMath|CheckHealthValue))
  (bind (_returnvalue_4 _currenthealth_4) (PureMath|CheckHealthValue))
  (if (< _currenthealth 75.0)
    (bind _targetitemcount (Shared|FindIteminInventory))
    (bind _output_get (LocalVariables|SetFoundItemCount _targetitemcount))
    ; Old
    (if (>= _output_get 2)
      (if true
        (bind _index (Utilities|FlowControl|ForLoopwithBreak 0 (- (LocalVariables|GetFoundItemCount) 1))
          (:LoopBody
            (Shared|DestroyIteminInventory 1)
            (Shared|ApplyHeal (Math|Float|SelectFloat (LocalVariables|GetHealthAmount) (Math|Float|SelectFloat (- (LocalVariables|GetHealthAmount) (- (+ (LocalVariables|GetHealthAmount) _currenthealth_4) 75.0)) (LocalVariables|GetHealthAmount) (>= (+ (LocalVariables|GetHealthAmount) _currenthealth_4) 75.0)) (== _index 0)))
            (if (>= _currenthealth_2 75.0)
              _))
          (:Completed
            _))
        (return)
        (else
          _)))
    (Shared|DestroyIteminInventory 1)
    (Shared|ApplyHeal (Math|Float|SelectFloat (- (LocalVariables|GetHealthAmount) (- (+ (LocalVariables|GetHealthAmount) _currenthealth_3) 75.0)) (LocalVariables|GetHealthAmount) (>= (+ (LocalVariables|GetHealthAmount) _currenthealth_3) 75.0)))
    (return)
    (else
      (return))))

(fn UseShieldingItem
  (bind (_returnvalue _currentshield) (PureMath|CheckShieldValue))
  (bind _targetitemcount (Shared|FindIteminInventory))
  (LocalVariables|SetFoundItemCount _targetitemcount)
  (Utilities|FlowControl|ForLoopwithBreak 0 (Math|Integer|Clamp(Integer) (- (LocalVariables|GetFoundItemCount) 1) 0 100)
    (:LoopBody
      (bind _shieldingamount (CallFunction|SetShieldingAmount))
      (Shared|ApplyShield _shieldingamount)
      (if (>= _currentshield (Variables|Default|GetShieldingCap))
        _))
    (:Completed
      _))
  (return))

(fn UseSmallShieldPotion
  (bind (_returnvalue _currentshield) (PureMath|CheckShieldValue))
  (bind (_returnvalue_2 _currentshield_2) (PureMath|CheckShieldValue))
  (bind (_returnvalue_3 _currentshield_3) (PureMath|CheckShieldValue))
  (bind (_returnvalue_4 _currentshield_4) (PureMath|CheckShieldValue))
  (if (< _currentshield_3 5e+01)
    (bind _targetitemcount (Shared|FindIteminInventory))
    (bind _output_get (LocalVariables|SetFoundItemCount _targetitemcount))
    (if (>= _output_get 2)
      (if (> (LocalVariables|GetShieldAmount) _currentshield)
        (for _index (range 2)
          (Shared|DestroyIteminInventory 1)
          (Shared|ApplyShield (Math|Float|SelectFloat (LocalVariables|GetShieldAmount) (Math|Float|SelectFloat (- (LocalVariables|GetShieldAmount) (- (+ (LocalVariables|GetShieldAmount) _currentshield_4) 5e+01)) (LocalVariables|GetShieldAmount) (>= (+ (LocalVariables|GetShieldAmount) _currentshield_4) 5e+01)) (== _index 0))))
        (return)
        (else
          _)))
    (Shared|DestroyIteminInventory 1)
    (Shared|ApplyShield (Math|Float|SelectFloat (- (LocalVariables|GetShieldAmount) (- (+ (LocalVariables|GetShieldAmount) _currentshield_2) 5e+01)) (LocalVariables|GetShieldAmount) (>= (+ (LocalVariables|GetShieldAmount) _currentshield_2) 5e+01)))
    (return)
    (else
      (return))))
